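# Material function that blends mesh material attributes into the Athena runtime virtual textures by world height relative to the RVT height, which it also outputs.
# Normals blend by distance, while color, specular and roughness use a HeightLerp mask; vertex normal Z shortens blends on sides and biases mips.
# An optional snow pass applies at certain quality levels.
# Static, RVT-output and shading-path switches bypass blending, including on forward and mobile.
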MaterialFunction /BlastBerryMapContent/Game/TechArt/Material_Functions/MF_Rufus_RVTBlendingOn
graph {nodes [
  n1 = NamedRerouteDeclaration(None=FunctionInput() {inputType=FunctionInput_MaterialAttributes, sortPriority=-1}) {name=Pre RVT, nodeColor=<rgba 1 0.7734 0 1>}
  n2 = MaterialFunctionCall(NoRVT=@n1, RVT=MaterialProxyReplace(Realtime=NamedRerouteUsage(), MaterialProxy=@n1)) {materialFunction=&FN_RVT_Switch_Material}
  FunctionOutput(None=StaticSwitchParameter(True=ShadingPathSwitch(Default=@n2.Result, Deferred=@n2.Result, Forward=@n1, Mobile=@n1), False=NamedRerouteUsage()) {defaultValue=true, group=D_RVT, parameterName=Use_RVT Blending, sortPriority=-1})
  n3 = RuntimeVirtualTextureSampleParameter() {
   group=D_RVT
   materialType=WorldHeight
   parameterName=RVTHeight
   sortPriority=1
   virtualTexture=/BlastBerryMapContent/Game/Athena/Environments/Materials/RVT/Athena_RuntimeVirtualTexture_Height
  }
  n4 = NamedRerouteDeclaration(None=@n3.WorldHeight) {name=RVT Height as is, nodeColor=<rgba 0 0.4141 1 1>}
  FunctionOutput(None=@n4) {outputName=RVT H, sortPriority=2}
  NamedRerouteDeclaration(None=GetMaterialAttributes(None=@n1) {attributeGetTypes [#1, #2]}.Roughness) {name=Roughness, nodeColor=<rgba 0.2188 0 1 1>}
  n5 = Constant() {r=1}
  n6 = NamedRerouteUsage()
  n7 = VertexNormalWS()
  n8 = RuntimeVirtualTextureSampleParameter!(UVs, World Position, Mip Bias=LinearInterpolate(Alpha=ComponentMask(None=@n7) {b=true}) {constA=6, constB=0}) {
   group=D_RVT
   materialType=BaseColor_Normal_Specular_YCoCg
   mipValueMode=RVTMVM_MipBias
   parameterName=RVT
   sortPriority=1
   virtualTexture=/BlastBerryMapContent/Game/Athena/Environments/Materials/RVT/Athena_RuntimeVirtualTexture
  }
  n9 = WorldPosition()
  n10 = Subtract(A=Add(A=ScalarParameter() {group=D_RVT, parameterName=Offset RVT WP, sortPriority=4}, B=ComponentMask(None=@n9.XYZ) {b=true}), B=@n4)
  n11 = NamedRerouteDeclaration(None=@n10) {name=RVT Height, nodeColor=<rgba 1 0 0.2344 1>}
  n12 = ScalarParameter() {
   defaultValue=10
   desc="Controls the blending of normals on sides of meshes, but not UP vector surfaces"
   group=D_RVT
   parameterName=Blend RVT Normals
   sortPriority=3
  }
  n13 = ScalarParameter() {defaultValue=4, group=D_RVT, parameterName=RVT Sides Divider}
  n14 = Abs(None=ComponentMask(None=VertexNormalWS()) {b=true})
  n15 = LinearInterpolate(A=Divide(A=@n12, B=@n13) {constB=4}, B=@n12, Alpha=@n14)
  n16 = BlendMaterialAttributes(
   A=SetMaterialAttributes!(MaterialAttributes=@n6, Normal=NamedRerouteDeclaration(None=@n8.Normal) {name=RVT N, nodeColor=<rgba 0 0.6953 1 1>}) {attributeSetTypes [#3]}
   B=@n6
   Alpha=Saturate(None=Divide(A=@n11, B=@n15))
  ) {vertexAttributeBlendType=UseA}
  n17 = SetMaterialAttributes!(
   MaterialAttributes=@n16
   BaseColor=NamedRerouteDeclaration(None=@n8.BaseColor) {name=RVT Base Color, nodeColor=<rgba 0 1 0.6016 1>}
   Specular=NamedRerouteDeclaration(None=@n8.Specular) {name=RVT Specular, nodeColor=<rgba 1 0 0.8203 1>}
   Roughness=NamedRerouteDeclaration(None=@n8.Roughness) {name=RVT Rough, nodeColor=<rgba 0 1 0.7422 1>}
   EmissiveColor=Constant()
  ) {attributeSetTypes [
    #4
    #2
    #1
    #5
   ]}
  n18 = MaterialFunctionCall(
   Transition Phase=NamedRerouteUsage()
   Height Texture=OneMinus(None=NamedRerouteUsage())
   Contrast=ScalarParameter() {group=D_RVT, parameterName=RVT_Contrast, sortPriority=5}
  ) {materialFunction=&HeightLerp}
  n19 = BlendMaterialAttributes(
   A=NamedRerouteDeclaration(None=@n17) {name=RVT Colors, nodeColor=<rgba 0.5938 1 0 1>}
   B=@n16
   Alpha=Saturate(None=NamedRerouteDeclaration(None=@n18.Alpha) {name=RVTMask, nodeColor=<rgba 0 0.9531 1 1>})
  ) {vertexAttributeBlendType=UseA}
  n20 = OneMinus(None=NamedRerouteUsage())
  n21 = StaticSwitchParameter(
   True=BlendMaterialAttributes(
    A=@n19
    B=MaterialFunctionCall!(Mask=@n5, SlipperySnow_Mask, "Material;"=@n19) {materialFunction=/Game/Athena/Asteria/Environments/Landscape/Materials/MaterialFunctions/MF_SnowPostRVT}.SnowMaterial
    Alpha=Saturate(None=@n20)
   )
   False=@n19
  ) {desc=Same as landscape. Should affect only rvt areas. Use onlty withing snowing biome., group=D_RVT, parameterName=RVT_SnowPostFunction}
  NamedRerouteDeclaration(None=QualitySwitch(Default=@n21, Low=@n19, High=@n19, Medium=@n19, Epic=@n21)) {name=RVT On, nodeColor=<rgba 0 0.8594 1 1>}
  n22 = ScalarParameter() {
   defaultValue=30
   desc="Used to control how far colors will blend with the rvt, as well sets blending distance of normals for Up looking surfaces"
   group=D_RVT
   parameterName=Blend RVT
   sortPriority=2
  }
  n23 = Divide(A=@n22, B=@n13) {constB=4}
  NamedRerouteDeclaration(None=Divide(A=@n10, B=LinearInterpolate(A=@n23, B=@n22, Alpha=@n14))) {name=RVT Blending, nodeColor=<rgba 0.9922 1 0 1>}
 ]}
